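# Translucent UI master material for player augment icons, with animatable color, glow and opacity.
# The Icon texture packs icon, backing plate and glow into R, G and B, with subtract-and-multiply thresholds sharpening the icon, outline and plate masks.
# Anim scalars blend start and stop values, while IsUnlockedState swaps color sets; the unlocked plate color has no separate stop.
# An IsFullRGBImage switch allows plain RGBA textures instead of packed masks.
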Material /Game/UI/Icons/Augments/M_UI_PlayerAugment_Icon_Master
graph {
 properties {
  bUsedWithStaticMesh=false
  blendMode=BLEND_Translucent
  materialDomain=MD_UI
  pixelDepthOffsetMode=PDOM_Legacy
  usedShadingModels=Default Lit
 }
 nodes [
  n1 = TextureSample(Tex=TextureObjectParameter() {parameterName=Icon, texture=/Game/Athena/UI/Frontend/Spatial/Art/T_Ui_Empty}) {samplerType=SAMPLERTYPE_LinearColor}
  NamedRerouteDeclaration(None=Saturate(None=@n1.B)) {name=Glow, nodeColor=<rgba 1 0.4453 0 1>}
  NamedRerouteDeclaration(None=@n1.RGB) {name=Full RBB Image, nodeColor=<rgba 1 0.5391 0 1>}
  NamedRerouteDeclaration(None=@n1.A) {name=Full RGB Image Alpha, nodeColor=<rgba 0.2422 1 0 1>}
  n2 = Saturate(None=@n1.R)
  n3 = Subtract(A=@n2, B=ScalarParameter() {defaultValue=0.25, group=Icon, parameterName=IconOutline_Subtract})
  n4 = ScalarParameter() {defaultValue=8, group=Icon, parameterName=IconOutline_Hardness}
  n5 = NamedRerouteDeclaration(None=Saturate(None=Multiply(A=@n3, B=@n4))) {name=IconOutline, nodeColor=<rgba 0.2187 1 0 1>}
  n6 = Subtract(A=Saturate(None=@n1.G), B=ScalarParameter() {defaultValue=0.5, group=Backing, parameterName=BackingPlate_Subtract})
  n7 = ScalarParameter() {defaultValue=8, group=Backing, parameterName=BackingPlate_Hardness}
  n8 = NamedRerouteDeclaration(None=Saturate(None=Multiply(A=@n6, B=@n7))) {name=BackingPlate, nodeColor=<rgba 0.2188 0 1 1>}
  NamedRerouteDeclaration(None=Saturate(None=Add(A=@n5, B=@n8))) {name=Icon + BackingPlate, nodeColor=<rgba 0.8281 1 0 1>}
  n9 = Subtract(A=@n2, B=ScalarParameter() {defaultValue=0.5, group=Icon, parameterName=Icon_Subtract})
  n10 = ScalarParameter() {defaultValue=8, group=Icon, parameterName=Icon_Hardness}
  NamedRerouteDeclaration(None=Saturate(None=Multiply(A=@n9, B=@n10))) {name=Icon, nodeColor=<rgba 0.5 1 0 1>}
  n11 = NamedRerouteUsage()
  n12 = NamedRerouteUsage()
  n13 = LinearInterpolate(
   A=LinearInterpolate(
    A=VectorParameter() {group=Color, parameterName=BackingPlate_Color_Unlocked, defaultValue=<rgba 0 1 0 1>}.RGB
    B=VectorParameter() {group=Color, parameterName=BackingPlate_Color_Unlocked, defaultValue=<rgba 0 1 0 1>}.RGB
    Alpha=ScalarParameter() {parameterName=Anim_IconColor}
   )
   B=LinearInterpolate(
    A=VectorParameter() {group=Color, parameterName=Icon_Color_Unlocked, defaultValue=<rgba 1 0 1 1>}.RGB
    B=VectorParameter() {group=Color, parameterName=Icon_Color_Stop_Unlocked, defaultValue=<rgba 1 0 0 1>}.RGB
    Alpha=ScalarParameter() {parameterName=Anim_IconColor}
   )
   Alpha=Saturate(None=Add(A=@n11, B=@n12))
  )
  n14 = NamedRerouteUsage()
  n15 = NamedRerouteUsage()
  n16 = LinearInterpolate(
   A=LinearInterpolate(
    A=VectorParameter() {group=Color, parameterName=BackingPlate_Color, defaultValue=<rgba 0 1 0 1>}.RGB
    B=VectorParameter() {group=Color, parameterName=BackingPlate_Color_Stop, defaultValue=<rgba 0 0 1 1>}.RGB
    Alpha=ScalarParameter() {parameterName=Anim_IconColor}
   )
   B=LinearInterpolate(
    A=VectorParameter() {group=Color, parameterName=Icon_Color, defaultValue=<rgba 1 0 1 1>}.RGB
    B=VectorParameter() {group=Color, parameterName=Icon_Color_Stop, defaultValue=<rgba 1 0 0 1>}.RGB
    Alpha=ScalarParameter() {parameterName=Anim_IconColor}
   )
   Alpha=Saturate(None=Add(A=@n14, B=@n15))
  )
  NamedRerouteDeclaration(None=StaticSwitchParameter(True=Saturate(None=@n13), False=Saturate(None=@n16)) {parameterName=IsUnlockedState}) {name=Color, nodeColor=<rgba 0 1 0.25 1>}
  n17 = ScalarParameter() {defaultValue=1, group=Anim, parameterName=Glow_Amount}
  n18 = ScalarParameter() {defaultValue=1, group=Anim, parameterName=Glow_Amount_Stop}
  n19 = ScalarParameter() {defaultValue=1, group=Anim, parameterName=Anim_GlowAmount}
  NamedRerouteDeclaration(None=Multiply(A=NamedRerouteUsage(), B=LinearInterpolate(A=@n17, B=@n18, Alpha=@n19))) {name=GlowAmount, nodeColor=<rgba 1 0 0.0469 1>}
  NamedRerouteDeclaration(None=Multiply(A=NamedRerouteUsage(), B=ScalarParameter() {defaultValue=1, group=Icon, parameterName=Icon_Alpha})) {name=IconAlpha, nodeColor=<rgba 1 0 0.9609 1>}
  n20 = NamedRerouteUsage()
  n21 = LinearInterpolate(
   A=ScalarParameter() {defaultValue=1, group=Backing, parameterName=BackingPlate_Alpha}
   B=ScalarParameter() {defaultValue=1, group=Backing, parameterName=BackingPlate_Alpha_Stop}
   Alpha=ScalarParameter() {parameterName=Anim_BackingAlpha}
  )
  NamedRerouteDeclaration(None=Saturate(None=Multiply(A=@n20, B=@n21))) {name=Backing Plate Alpha, nodeColor=<rgba 1 0.8438 0 1>}
  n22 = LinearInterpolate(A=Add(A=NamedRerouteUsage(), B=NamedRerouteUsage()), B=NamedRerouteUsage(), Alpha=NamedRerouteUsage())
  n23 = LinearInterpolate(
   A=ScalarParameter() {defaultValue=1, group=Alpha, parameterName=Final_Alpha_Start}
   B=ScalarParameter() {defaultValue=1, group=Alpha, parameterName=Final_Alpha_Stop}
   Alpha=ScalarParameter() {group=Anim, parameterName=Anim_FinalAlpha}
  )
  NamedRerouteDeclaration(None=Saturate(None=Multiply(A=@n22, B=@n23))) {name=Final Alpha, nodeColor=<rgba 0 1 0.9063 1>}
  OneMinus(None=@n8)
  StaticSwitchParameter(True=NamedRerouteUsage(), False=NamedRerouteUsage()) {parameterName=IsFullRGBImage}
  StaticSwitchParameter(True=NamedRerouteUsage(), False=NamedRerouteUsage()) {parameterName=IsFullRGBImage}
 ]
}
